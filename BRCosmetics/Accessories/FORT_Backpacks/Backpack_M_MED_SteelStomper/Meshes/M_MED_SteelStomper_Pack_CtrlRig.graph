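# Control Rig for the Steel Stomper back bling, keeping its two pneumatic neck cylinders pointed at each other as the robot neck moves.
# Four chained AimItem units rotate each upper and lower cylinder bone toward the opposite end's aim bone, with the primary axis flipped on the left side.
# Separate aim bones coincide with each cylinder bone and share its parent, giving targets that the aiming does not rotate.
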ControlRigBlueprint /BRCosmetics/Accessories/FORT_Backpacks/Backpack_M_MED_SteelStomper/Meshes/M_MED_SteelStomper_Pack_CtrlRig
graphs [
 RigVMFunctionLibrary {links [], nodes []}
 RigVMModel {nodes [
   RigUnit_BeginExecution = &RigUnit_BeginExecution()
   n1 = &RigUnit_GetTransform(Item=(Type=Bone,Name="pneuCyl_up_R_aim"), bInitial=False)
   /Script/ControlRig.RigUnit_AimItem(
    DebugSettings=(bEnabled=False,Scale=10,WorldOffset=(Rotation=(X=0,Y=0,Z=0,W=1),Translation=(X=0,Y=0,Z=0),Scale3D=(X=1,Y=1,Z=1)))
    Item=(Type=Bone,Name="pneuCyl_lo_R")
    Primary=(Weight=1,Axis=(X=1,Y=0,Z=0),Target=(X=1,Y=0,Z=0),Kind=Location,Space=(Type=Bone,Name="None"))
    Secondary=(Weight=1,Axis=(X=0,Y=0,Z=1),Target=(X=1,Y=0,Z=0),Kind=Location,Space=(Type=Bone,Name="None"))
    Weight=1
    ExecutePin=@RigUnit_BeginExecution.ExecutePin
    Primary.Target=@n1.Transform.Translation
    Secondary.Target=@n1.Transform.Translation
   ) {is=struct RigVMUnitNode}
   n2 = &RigUnit_GetTransform(Item=(Type=Bone,Name="pneuCyl_lo_R_aim"), bInitial=False)
   > &RigUnit_AimItem(
    DebugSettings=(bEnabled=False,Scale=10,WorldOffset=(Rotation=(X=0,Y=0,Z=0,W=1),Translation=(X=0,Y=0,Z=0),Scale3D=(X=1,Y=1,Z=1)))
    Item=(Type=Bone,Name="pneuCyl_up_R")
    Primary=(Weight=1,Axis=(X=1,Y=0,Z=0),Target=(X=1,Y=0,Z=0),Kind=Location,Space=(Type=Bone,Name="None"))
    Secondary=(Weight=1,Axis=(X=0,Y=0,Z=1),Target=(X=1,Y=0,Z=0),Kind=Location,Space=(Type=Bone,Name="None"))
    Weight=1
    Primary.Target=@n2.Transform.Translation
    Secondary.Target=@n2.Transform.Translation
   )
   n3 = &RigUnit_GetTransform(Item=(Type=Bone,Name="pneuCyl_up_L_aim"), bInitial=False)
   > &RigUnit_AimItem(
    DebugSettings=(bEnabled=False,Scale=10,WorldOffset=(Rotation=(X=0,Y=0,Z=0,W=1),Translation=(X=0,Y=0,Z=0),Scale3D=(X=1,Y=1,Z=1)))
    Item=(Type=Bone,Name="pneuCyl_lo_L")
    Primary=(Weight=1,Axis=(X=-1,Y=0,Z=0),Target=(X=1,Y=0,Z=0),Kind=Location,Space=(Type=Bone,Name="None"))
    Secondary=(Weight=1,Axis=(X=0,Y=0,Z=1),Target=(X=1,Y=0,Z=0),Kind=Location,Space=(Type=Bone,Name="None"))
    Weight=1
    Primary.Target=@n3.Transform.Translation
    Secondary.Target=@n3.Transform.Translation
   )
   RigUnit_GetTransform = &RigUnit_GetTransform(Item=(Type=Bone,Name="pneuCyl_lo_L_aim"), bInitial=False)
   > &RigUnit_AimItem(
    DebugSettings=(bEnabled=False,Scale=10,WorldOffset=(Rotation=(X=0,Y=0,Z=0,W=1),Translation=(X=0,Y=0,Z=0),Scale3D=(X=1,Y=1,Z=1)))
    Item=(Type=Bone,Name="pneuCyl_up_L")
    Primary=(Weight=1,Axis=(X=-1,Y=0,Z=0),Target=(X=1,Y=0,Z=0),Kind=Location,Space=(Type=Bone,Name="None"))
    Secondary=(Weight=1,Axis=(X=0,Y=0,Z=1),Target=(X=1,Y=0,Z=0),Kind=Location,Space=(Type=Bone,Name="None"))
    Weight=1
    Primary.Target=@RigUnit_GetTransform.Transform.Translation
    Secondary.Target=@RigUnit_GetTransform.Transform.Translation
   )
  ]}
]
hierarchy [
 root {loc=<0 0 0>, rot=<0 0 0 1>}
 pack {parent=root, loc=<0 -0.7316 0>, rot=<0 0 0 1>}
 roboBase {parent=pack, loc=<0 -8.1709 -5.6126>, rot=<-0.0018 0.7071 0.0018 -0.7071>}
 roboNeck_01 {parent=roboBase, loc=<0 0 0>, rot=<0 0 0 1>}
 roboNeck_02 {parent=roboNeck_01, loc=<4.324 0 0>, rot=<0.0018 -0.7071 -0.0018 -0.7071>}
 roboNeck_03 {parent=roboNeck_02, loc=<0 -0.0224 4.3236>, rot=<0 0 0 1>}
 flapBase_R {parent=roboNeck_03, loc=<6.3714 -0.7167 3.7074>, rot=<0.0011 0.0218 -0.0229 -0.9995>}
 flapJoint_R {parent=flapBase_R, loc=<5.6858 -0.4207 0.0874>, rot=<0 0 0.0281 -0.9996>}
 flap_R {parent=flapJoint_R, loc=<0 0 0>, rot=<0.0164 -0.6153 -0.0007 -0.7881>}
 hose_master {parent=roboNeck_03, loc=<-0.0221 5.601 2.6172>, rot=<-0.2692 -0.8352 -0.1471 -0.4564>}
 dyn_hose_01 {parent=hose_master, loc=<0 0 0>, rot=<0.9968 -0.0291 0.0045 0.0743>}
 dyn_hose_02 {parent=dyn_hose_01, loc=<3.6088 0 0>, rot=<0.0605 0.1063 -0.2856 -0.9505>}
 dyn_hose_03 {parent=dyn_hose_02, loc=<3.4625 0 0>, rot=<0.0033 -0.0373 -0.0287 -0.9989>}
 dyn_hose_04 {parent=dyn_hose_03, loc=<4.5954 0 0>, rot=<-0.0054 -0.0486 -0.0416 -0.9979>}
 dyn_hose_05 {parent=dyn_hose_04, loc=<4.7716 0 0>, rot=<-0.0084 -0.0681 -0.0361 -0.997>}
 dyn_hose_06 {parent=dyn_hose_05, loc=<4.0261 0 0>, rot=<0.0163 -0.0294 -0.1075 -0.9936>}
 dyn_hose_07 {parent=dyn_hose_06, loc=<3.697 0 0>, rot=<0.042 0.0525 -0.3685 -0.9272>}
 dyn_hose_08 {parent=dyn_hose_07, loc=<4.653 0 0>, rot=<-0.0061 0.1093 -0.4397 -0.8915>}
 dyn_hose_09 {parent=dyn_hose_08, loc=<3.1836 0 0>, rot=<-0.0405 0.0993 -0.4266 -0.898>}
 dyn_hose_10 {parent=dyn_hose_09, loc=<3.7733 0 0>, rot=<0 0 0 1>}
 pneuCyl_up_R {parent=roboNeck_03, loc=<2.9279 -2.4852 1.1113>, rot=<-0.444 -0.4495 0.5747 -0.5201>}
 eye_R {parent=roboNeck_03, loc=<2.6149 -4.7917 5.9989>, rot=<0.0328 0.1246 0.6501 -0.7489>}
 eye_L {parent=roboNeck_03, loc=<-2.6149 -4.7917 5.9989>, rot=<0.7489 0.6501 -0.1246 0.0328>}
 pneuCyl_up_L {parent=roboNeck_03, loc=<-2.9279 -2.4853 1.1113>, rot=<-0.5201 -0.5747 -0.4495 0.444>}
 flapBase_L {parent=roboNeck_03, loc=<-6.3714 -0.7167 3.7074>, rot=<0.9995 -0.0229 -0.0218 0.0011>}
 flapJoint_L {parent=flapBase_L, loc=<-5.6858 0.4207 -0.0874>, rot=<0 0 0.0281 -0.9996>}
 flap_L {parent=flapJoint_L, loc=<0 0 0>, rot=<0.0164 -0.6153 -0.0007 -0.7881>}
 pneuCyl_up_R_aim {parent=roboNeck_03, loc=<2.9279 -2.4852 1.1113>, rot=<-0.444 -0.4495 0.5747 -0.5201>}
 pneuCyl_up_L_aim {parent=roboNeck_03, loc=<-2.9279 -2.4853 1.1113>, rot=<-0.5201 -0.5747 -0.4495 0.444>}
 pneuCyl_lo_R {parent=roboBase, loc=<0.3135 -4.451 -2.3006>, rot=<-0.6854 -0.0942 -0.0517 -0.7202>}
 pneuCyl_lo_L {parent=roboBase, loc=<0.3135 -4.451 2.3006>, rot=<-0.0942 0.6854 0.7202 -0.0517>}
 pneuCyl_lo_L_aim {parent=roboBase, loc=<0.3135 -4.451 2.3006>, rot=<-0.0942 0.6854 0.7202 -0.0517>}
 pneuCyl_lo_R_aim {parent=roboBase, loc=<0.3135 -4.451 -2.3006>, rot=<-0.6854 -0.0942 -0.0517 -0.7202>}
 isPhysicsDisabled {type=Curve, loc=<0 0 0>, rot=<0 0 0 1>}
]
