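# Physics asset for the Steel Stomper backpack cosmetic, modelling its hose as a chain of small capsule bodies.
# Segments 01 to 09 chain from hose_master, which is angularly locked to roboBase, with tight swing and twist limits that loosen toward the far end.
# Segment 10 constrains directly to roboBase rather than segment 09.
# Three boxes on roboBase and a large unconstrained pack capsule provide the body shapes.
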PhysicsAsset /BRCosmetics/Accessories/FORT_Backpacks/Backpack_M_MED_SteelStomper/Meshes/M_MED_SteelStomper_Pack_Physics
bodies {
 dyn_hose_01 [{shapeType=Capsule, length=2.024, radius=1.0891, shapeName=dyn_hose_01_capsule, center=<2.0075 0 0>, rotation=<rot -90 180 179.9997>}]
 dyn_hose_02 [{shapeType=Capsule, length=2.024, radius=1.0891, shapeName=dyn_hose_02_capsule, center=<2.0075 0 0>, rotation=<rot -90 -179.9605 179.9604>}]
 dyn_hose_03 [{shapeType=Capsule, length=2.5356, radius=1.0891, shapeName=dyn_hose_03_capsule, center=<2.35 0 0>, rotation=<rot -90 -160.5288 160.5286>}]
 dyn_hose_04 [{shapeType=Capsule, length=2.5696, radius=1.0891, shapeName=dyn_hose_04_capsule, center=<2.2279 0 0>, rotation=<rot -90 -179.9999 179.9996>}]
 dyn_hose_05 [{shapeType=Capsule, length=2.024, radius=1.0891, shapeName=dyn_hose_05_capsule, center=<2.0075 0 0>, rotation=<rot -90 0 -0.0004>}]
 dyn_hose_06 [{shapeType=Capsule, length=2.024, radius=1.0891, shapeName=dyn_hose_06_capsule, center=<2.0075 0 0>, rotation=<rot -90 0 -0.0003>}]
 dyn_hose_07 [{shapeType=Capsule, length=2.8904, radius=1.2946, shapeName=dyn_hose_07_capsule, center=<2.2321 -0.6232 0>, rotation=<rot -90 6.7215 -6.7221>}]
 dyn_hose_08 [{shapeType=Capsule, length=2.4807, radius=1.0891, shapeName=dyn_hose_08_capsule, center=<2.0075 -0.5885 0>, rotation=<rot -90 90 -90.0006>}]
 dyn_hose_09 [{
   shapeType=Capsule
   length=1.7916
   radius=1.3215
   shapeName=dyn_hose_09_capsule
   center=<2.0075 -0.3229 0>
   rotation=<rot -90 80.7989 -76.3371>
  }]
 dyn_hose_10 [{shapeType=Capsule, length=0.1152, radius=0.5482, shapeName=dyn_hose_10_capsule, center=<2.0075 0 0>, rotation=<rot -90 160.529 -160.5297>}]
 hose_master [{shapeType=Sphere, radius=0.5, shapeName=hose_master_sphere}]
 pack [{shapeType=Capsule, length=15, radius=15, shapeName=pack_capsule, center=<0 15.5786 0>}]
 roboBase [
  {shapeType=Box, extentX=7.5, extentY=1.2276, extentZ=14.8685, shapeName=roboBase_box, center=<3.7881 -7.7662 0>, rotation=<rot 0 -5.1981 0>}
  {
   shapeType=Box
   extentX=7.5
   extentY=1.2276
   extentZ=14.8685
   shapeName=roboBase_box
   center=<3.166 -1.5139 8.9933>
   rotation=<rot 17.8248 0.1139 90>
  }
  {
   shapeType=Box
   extentX=7.5
   extentY=1.2276
   extentZ=14.8685
   shapeName=roboBase_box
   center=<3.5094 -1.5331 -8.7985>
   rotation=<rot -19.6631 0.1139 90>
  }
 ]
}
constraints [
 {bone1Name=dyn_hose_01, bone2Name=hose_master, swing1LimitDegrees=5, swing2LimitDegrees=5, twistLimitDegrees=5}
 {bone1Name=dyn_hose_02, bone2Name=dyn_hose_01, swing1LimitDegrees=5, swing2LimitDegrees=5, twistLimitDegrees=1}
 {bone1Name=dyn_hose_03, bone2Name=dyn_hose_02, swing1LimitDegrees=5, swing2LimitDegrees=5, twistLimitDegrees=1}
 {bone1Name=dyn_hose_04, bone2Name=dyn_hose_03, swing1LimitDegrees=5, swing2LimitDegrees=5, twistLimitDegrees=1}
 {bone1Name=dyn_hose_05, bone2Name=dyn_hose_04, swing1LimitDegrees=5, swing2LimitDegrees=5, twistLimitDegrees=1}
 {bone1Name=dyn_hose_06, bone2Name=dyn_hose_05, swing1LimitDegrees=5, swing2LimitDegrees=5, twistLimitDegrees=1}
 {bone1Name=dyn_hose_07, bone2Name=dyn_hose_06, swing1LimitDegrees=5, swing2LimitDegrees=5, twistLimitDegrees=1}
 {bone1Name=dyn_hose_08, bone2Name=dyn_hose_07, swing1LimitDegrees=15, swing2LimitDegrees=15, twistLimitDegrees=5}
 {bone1Name=dyn_hose_09, bone2Name=dyn_hose_08, swing1LimitDegrees=15, swing2LimitDegrees=15, twistLimitDegrees=5}
 {bone1Name=dyn_hose_10, bone2Name=roboBase, swing1LimitDegrees=15, swing2LimitDegrees=15, twistLimitDegrees=5}
 {
  bone1Name=hose_master
  bone2Name=roboBase
  swing1LimitDegrees=45
  swing1Motion=Locked
  swing2LimitDegrees=45
  swing2Motion=Locked
  twistLimitDegrees=45
  twistMotion=Locked
 }
]
