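# Animation Blueprint for the SteelStomper backpack cosmetic, playing idle and damage-reaction animations.
# A state machine plays random idles and switches to Damage on TookDamage, set by health or shield input events or a delayed front-end trigger.
# The pose passes through an additive pack-bone transform, RigidBody physics and a Control Rig.
# The event graph computes physics weight from a curve, skydive gravity override, and pack offsets from owner spine yaw.
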AnimBlueprint /BRCosmetics/Accessories/FORT_Backpacks/Backpack_M_MED_SteelStomper/Meshes/M_MED_SteelStomper_Pack_AnimBP
target_skeleton=&/M_MED_SteelStomper_Pack_Skeleton
graphs [
 AnimGraph {nodes [
   Input Pose()
   AnimGraphNode_StateMachine_0 = Backpack States() {is=title AnimGraphNode_StateMachine}
   n1 = "Transform (Modify) Bone - Bone: pack"(ComponentPose=Local To Component(LocalPose=@AnimGraphNode_StateMachine_0), Translation="0,0,0", Rotation="0,0,0") {rotationMode=BMM_Additive, translationMode=BMM_Additive, alphaBoolBlend {blendOption=Linear}, boneToModify {boneName=pack}}
   n2 = RigidBody!(ComponentPose=@n1, Alpha, OverrideWorldGravity:Vector="0,0,0", ExternalForce) -> (Pose) {
    bClampLinearTranslationLimitToRefPose=true
    bOverrideWorldGravity=true
    cachedBoundsScale=1.1
    overridePhysicsAsset=&/M_MED_SteelStomper_Pack_Physics
    alphaBoolBlend {blendOption=Linear}
    componentLinearAccScale=<0.2 0.4 0.4>
    componentLinearVelScale=<0.2 0.4 0.4>
   }
   n3 = Control Rig(Source=Component To Local(ComponentPose=@n2)) {
    alphaBoolBlend {blendOption=Linear}
    controlRigAssetReference {blueprintRigClass=&/M_MED_SteelStomper_Pack_CtrlRig.M_MED_SteelStomper_Pack_CtrlRig_C}
   }
   Output Pose(Result=@n3) {layerGroup=None}
  ]}
 Backpack States {
  entry=Idle_Random
  path=AnimGraph.AnimGraphNode_StateMachine_0.Backpack States
  ids=id
  nodes [
   AnimStateNode_0 = Idle_Random() {
    is=title AnimStateNode
    stateEntered {guid=#1, notifyName=Entered_RandomLocoBreak}
    stateFullyBlended {guid=#2}
    stateLeft {guid=#3}
   }
   AnimStateNode_2 = IdleBreak() {is=title AnimStateNode}
   AnimStateNode_4 = idle() {is=title AnimStateNode}
   AnimStateNode_5 = Damage() {is=title AnimStateNode}
   AnimStateNode_6 = Backpack_SteelStomper_Damage() {is=title AnimStateNode}
   AnimStateTransitionNode_6 = AnimStateTransitionNode()
   AnimStateTransitionNode_7 = AnimStateTransitionNode() {bAutomaticRuleBasedOnSequencePlayerInState=true}
  ]
  links [
   {
    dst=Idle_Random
    rule=AnimGraph.AnimGraphNode_StateMachine_0.Backpack States.AnimStateTransitionNode_7.Transition
    src=Damage
    via=AnimStateTransitionNode_7
   }
   {
    dst=Damage
    rule=AnimGraph.AnimGraphNode_StateMachine_0.Backpack States.AnimStateTransitionNode_6.Transition
    src=Idle_Random
    via=AnimStateTransitionNode_6
   }
  ]
 }
 Idle_Random {
  path=AnimGraph.AnimGraphNode_StateMachine_0.Backpack States.AnimStateNode_0.Idle_Random
  nodes [
   n1 = Random Sequence Player!() -> (Pose:Pose Link Structure) {
    is=title AnimGraphNode_RandomPlayer
    entries [
     {
      chanceToPlay=0.8
      maxLoopCount=0
      maxPlayRate=1
      minLoopCount=0
      minPlayRate=1
      sequence=/BRCosmetics/Accessories/FORT_Backpacks/Backpack_M_MED_SteelStomper/Anim/Backpack_SteelStomper_Idle
      blendIn {blendOption=Linear, blendTime=0.2, customCurve=null}
     }
     {
      chanceToPlay=0.3
      maxLoopCount=0
      maxPlayRate=1
      minLoopCount=0
      minPlayRate=1
      sequence=/BRCosmetics/Accessories/FORT_Backpacks/Backpack_M_MED_SteelStomper/Anim/Backpack_SteelStomper_IdleBreak
      blendIn {blendOption=Linear, blendTime=0.2, customCurve=null}
     }
    ]
   }
   Output Animation Pose(Result=@n1)
  ]
 }
 IdleBreak {
  path=AnimGraph.AnimGraphNode_StateMachine_0.Backpack States.AnimStateNode_2.IdleBreak
  nodes [
   n1 = Sequence Player 'Backpack_SteelStomper_IdleBreak'() {bLoopAnimation=false, sequence=&Backpack_SteelStomper_IdleBreak}
   Output Animation Pose(Result=@n1)
  ]
 }
 idle {
  path=AnimGraph.AnimGraphNode_StateMachine_0.Backpack States.AnimStateNode_4.idle
  nodes [Output Animation Pose(Result=Sequence Player 'Backpack_SteelStomper_Idle'() {sequence=&Backpack_SteelStomper_Idle})]
 }
 Damage {
  path=AnimGraph.AnimGraphNode_StateMachine_0.Backpack States.AnimStateNode_5.Damage
  nodes [
   n1 = Sequence Player 'Backpack_SteelStomper_Damage'() {bLoopAnimation=false, sequence=/BRCosmetics/Accessories/FORT_Backpacks/Backpack_M_MED_SteelStomper/Anim/Backpack_SteelStomper_Damage}
   Output Animation Pose(Result=@n1)
  ]
 }
 Backpack_SteelStomper_Damage {
  path=AnimGraph.AnimGraphNode_StateMachine_0.Backpack States.AnimStateNode_6.Backpack_SteelStomper_Damage
  nodes [
   n1 = Sequence Player 'Backpack_SteelStomper_Damage'() {bLoopAnimation=false, sequence=&Backpack_SteelStomper_Damage}
   Output Animation Pose(Result=@n1)
  ]
 }
 Transition {
  path=AnimGraph.AnimGraphNode_StateMachine_0.Backpack States.AnimStateTransitionNode_6.Transition
  nodes [Result(bCanEnterTransition=Get TookDamage() -> (TookDamage:bool))]
 }
 Transition {path=AnimGraph.AnimGraphNode_StateMachine_0.Backpack States.AnimStateTransitionNode_7.Transition, nodes [Result()]}
 EventGraph {nodes [
   Event BlueprintUpdateAnimation()
   > Set isPhysics(isPhysics:float=Remap Float(Input=GetCurveValue(CurveName=isPhysicsDisabled), OldRange=(X=0,Y=1), NewRange=(X=1,Y=0), Clamp=true) -> (Output=~)) -> (Output_Get:float)
   > n1 = Gravity Override State Function DEPRECATED(CharacterPartAnimInstance=Self-Reference(), GameTimeInSeconds=GetGameTimeInSeconds(), GravityOverrideParams=Get Gravity Override Params())
   > Set Gravity Override(Gravity Override=@n1)
   n2 = GetOwnerSkeletalMeshComponent()
   n3 = BreakRotator(InRot=BreakTransform(InTransform=GetDeltaTransformFromRefPose(self=@n2, BoneName=spine_05, BaseName=spine_01)).Rotation)
   > Set BackpackRotate(BackpackRotate:Rotator=MakeRotator(Roll=MapRangeClamped(Value=@n3.Yaw, InRangeA=-8, InRangeB=-30, OutRangeB=30))) -> (Output_Get:Rotator)
   > Set BackpackTranslation(BackpackTranslation:Vector=MakeVector(Y=MapRangeClamped(Value=@n3.Yaw, InRangeA=-8, InRangeB=-30, OutRangeB=-5))) -> (Output_Get:Vector)
   Event BlueprintPostEvaluateAnimation() {is=title K2Node_Event}
   > Set TookDamage(TookDamage:bool=false) -> (Output_Get:bool)
   AnimNotify_ReactiveReset() {is=title K2Node_Event}
   > n4 = Set TookDamage(TookDamage=false)
   Branch(execute=[@n5.then, @n4.then], Condition=Get bIsInFrontEnd())
   > Delay(Duration=10)
   > Set TookDamage(TookDamage=true)
   n5 = Event BlueprintInitializeAnimation()
   n6 = Event BlueprintProcessAnimInputEvent()
   > Branch(Condition=OR Boolean(
     A=IsNamed(self=@n6.AnimInputEvent, InEventName=aie.PlayerVitality.HealthDecreased)
     B=IsNamed(self=@n6.AnimInputEvent, InEventName=Noneaie.PlayerVitality.ShieldsDecreased)
    ))
   > Set TookDamage(TookDamage=true)
   > RetriggerableDelay(Duration=0.2)
   > Set TookDamage(TookDamage=false)
   TryGetPawnOwner()
  ]}
]
variables [
 BackpackRotate {type=Rotator}
 BackpackTranslation {type=Vector}
 Gravity Override {type=Vector}
 Gravity Override Params {
  type=GravityOverrideParamsStruct
  default {
   skydive_DivingDown {gravityOverride=<-300 0 0>, windAmplitude=<50 150 100>}
   skydive_FloatingDown {gravityOverride=<0 -350 300>, windAmplitude=<100 -250 200>, windOffset=<0 -300 200>}
   skydive_GoingUp {gravityOverride=<0 -100 300>, windAmplitude=<100 -250 200>}
   skydive_Parachuting {gravityOverride=<0 -400 -490>, windAmplitude=<100 -100 100>}
  }
 }
 isPhysics {type=number, default=1}
 TookDamage {type=boolean}
]
